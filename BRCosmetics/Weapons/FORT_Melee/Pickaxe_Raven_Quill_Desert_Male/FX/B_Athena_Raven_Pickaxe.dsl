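; Cosmetic weapon Blueprint for the Raven pickaxe that drives its idle particle effect during melee swings.
; On cosmetic init it enables custom depth on the idle VFX component and copies the stencil value from player 0's character mesh.
; Left and right swing starts set the particle's "BladeSwing" bool true through OnToggleBladeMovement, and swing ends clear it after a short delay.
; Each overridden event calls its parent first.
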(event EventOnInitWeaponCosmetics
  (bind _returnvalue (CallFunction|GetIdleVFXComponent))
  (|CallParentFunction)
  (Rendering|SetRenderCustomDepth _returnvalue true)
  (Rendering|SetCustomDepthStencilValue _returnvalue (Variables|/Script/Engine.PrimitiveComponent|GetCustomDepthStencilValue (Variables|/Script/Engine.Character|GetMesh (Game|GetPlayerCharacter 0)))))

(event EventSwingLeftEnd
  (|CallParentFunction)
  (Utilities|FlowControl|Delay 2.0)
  (CallFunction|OnToggleBladeMovement false))

(event EventSwingRightEnd
  (|CallParentFunction)
  (Utilities|FlowControl|Delay 2.0)
  (CallFunction|OnToggleBladeMovement false))

(event EventSwingLeft
  (|CallParentFunction)
  (CallFunction|OnToggleBladeMovement true))

(event EventSwingRight
  (|CallParentFunction)
  (CallFunction|OnToggleBladeMovement true))

(fn OnToggleBladeMovement
  (params (InValue :type Boolean))
  (Effects|Components|ParticleSystem|SetBoolParameter (CallFunction|GetIdleVFXComponent) "BladeSwing" InValue))

(fn ConstructionScript
  (|CallParentFunction))
