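# Niagara idle flame effect for the Raven Quill Desert Male pickaxe, attached through the User skeletal mesh.
# Two mesh emitters pin layered flame meshes to sockets, feeding an interpolated position into per-particle data with a WPO amount, apparently for trailing lag.
# Their opacity scales with time of day.
# Sprite emitters add rising flame puffs and curl-noise ashes, halving and stopping their spawning respectively during BladeSwing.
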NiagaraSystem /BRCosmetics/Weapons/FORT_Melee/Pickaxe_Raven_Quill_Desert_Male/FX/NS_Pickaxe_Raven_Quill_Desert_Idle
dependencies {
 usedDataInterfaces [&NiagaraDataInterfaceVectorCurve, &NiagaraDataInterfaceColorCurve, &NiagaraDataInterfaceCurve]
 usedDynamicInputs []
 usedModules [&SystemState]
}
emitters {
 Flame_Ashes {
  emitterUpdateScript [
   &EmitterState()
   &SpawnRate(
    SpawnRate=&Multiply_Float(A=12, B=&MakeCustomFloatFromBool(Bool=$User.BladeSwing, True Float=0, False Float=1))
    Use Spawn Probability=false
   )
  ]
  particleSpawnScript [
   /Niagara/Modules/Spawn/Initialization/InitializeParticle(
    Color=<rgba 0.05 0.05 0.05 0.35>
    Lifetime=&UniformRangedFloat(Minimum=0.4, Maximum=0.8)
    Module.Write Color=true
    Module.Write Lifetime=true
    Module.Write Mass=false
    Module.Write Material Random=true
    Module.Write Position=false
    Module.Write Scale=false
    Module.Write SpriteRotation=true
    Module.Write SpriteSize=true
    Position=&SimulationPosition()
    Sprite Rotation=&UniformRangedFloat(Minimum=0, Maximum=360)
    Sprite Size=&RandomRangeVector2D(Minimum=<2 2>, Maximum=<4 4>)
   )
   /SetVariables(Particles.SpawnOffset:Vector3f=&MakeVector(
     X=&RandomRangeFloat(Minimum=-20, Maximum=20)
     Y=&RandomRangeFloat(Minimum=-5, Maximum=5)
     Z=&RandomRangeFloat(Minimum=-20, Maximum=0)
    )) {isSetParameters=true, module=SetVariables_#1}
   &SkeletalMeshLocation(
    Interpolate Mesh Sampling=false
    MaskSpawnGroup=false
    Mesh Sampling Type=::NewEnumerator5
    Position Sampling=::NewEnumerator1
    Sampled Position Offset=$Particles.SpawnOffset
    Skeletal Mesh=$User.SkeletalMesh
    Socket Sampling Mode=::NewEnumerator5
   )
   &CameraOffset(Camera Offset Amount=&RandomRangeFloat(Minimum=-2.5, Maximum=0))
   &AddVelocity(Velocity=<0 0 100>)
  ]
  particleUpdateScript [
   &self:AlphaBasedDistance(
    EmitterPosition:Vector3f=$Particles.Initial.SkeletalMeshLocation.SampledPosition
    MaxDistance=1
    ParticlePosition:NiagaraPosition=$Particles.Initial.Position
   ) {module=AlphaBasedDistance}
   &ParticleState()
   &CurlNoiseForce(Noise Strength=50, Pan Noise=false)
   &ScaleColor(
    Scale Alpha=&FloatFromCurve(
     FloatCurve=DataInterface() {propertyValues {Curve {keys [
         <at 0 0 RCIM_Cubic>
         {arriveTangent=0.1462, interpMode=RCIM_Cubic, leaveTangent=0.1462, time=0.169, value=1.0042}
         {arriveTangent=-1.2044, interpMode=RCIM_Cubic, leaveTangent=-1.2044, time=0.7656, value=0.1119}
         <at 1.0011 0.002>
        ]}}}
     Scale Curve=&Lerp_Float(A=1, B=0, Alpha=$Output.AlphaBasedDistance.AlphaScale)
    )
    Scale RGB=&VectorFromCurve(VectorCurve=DataInterface() {propertyValues {
       XCurve {keys [<at -0.0011 1.0067>,<at 0.8 0.9431>]}
       YCurve {keys [<at -0.0011 1.0067>,<at 0.7985 0.3834>]}
       ZCurve {keys [<at -0.0011 1.0067>,<at 0.7995 0.1189>]}
      }})
    ScaleA=true
    ScaleRGB=true
   )
   /SetVariables(Particles.SubImageIndex=&Add_Float(
     A=&FloatFromCurve(FloatCurve=DataInterface() {propertyValues {Curve {keys [<at 0 0>,<at 1 3>]}}})
     B=&RandomRangeFloat(Minimum=8, Maximum=30)
    )) {isSetParameters=true, module=SetVariables_#2}
   &ScaleSpriteSize(Scale Factor=&Vector2DFromFloat(Value=&FloatFromCurve(FloatCurve=DataInterface() {propertyValues {Curve {keys [
          {arriveTangent=7.5764, interpMode=RCIM_Cubic, leaveTangent=7.5765, tangentMode=RCTM_User, time=0, value=0.1}
          {arriveTangent=0.1045, interpMode=RCIM_Cubic, leaveTangent=0.1045, tangentMode=RCTM_User, time=0.25, value=1}
          {arriveTangent=-0.8944, interpMode=RCIM_Cubic, leaveTangent=-0.8944, tangentMode=RCTM_User, time=1, value=0.8}
         ]}}})))
   &Drag(Drag=4, Use Linear Drag=true, Use Rotational Drag=false)
   &SolveForcesAndVelocity(Clamp Velocity=false, Limit Acceleration=false)
   &SpriteRotationRate(Rotation Rate=250)
  ]
  rendererClasses [&NiagaraSpriteRendererProperties]
  renderers [{rendererClass=&NiagaraSpriteRendererProperties}]
 }
 Flame_Particles {
  emitterUpdateScript [
   &EmitterState()
   &SpawnRate(
    SpawnRate=&Multiply_Float(A=12, B=&MakeCustomFloatFromBool(Bool=$User.BladeSwing, True Float=0.5, False Float=1))
    Use Spawn Probability=false
   )
  ]
  particleSpawnScript [
   &InitializeParticle(
    Lifetime=&UniformRangedFloat(Minimum=0.15, Maximum=0.5)
    Module.Write Color=true
    Module.Write Lifetime=true
    Module.Write Mass=false
    Module.Write Material Random=true
    Module.Write Position=false
    Module.Write Scale=false
    Module.Write SpriteRotation=true
    Module.Write SpriteSize=true
    Position=&SimulationPosition()
    Sprite Rotation=&UniformRangedFloat(Minimum=0, Maximum=360)
    Sprite Size=&MakeCustomVector2DFromBool(
     Bool=&RandomBool()
     True Vector2D=&Vector2DFromFloat(Value=&RandomRangeFloat(Minimum=32, Maximum=64))
     False Vector2D=&Vector2DFromFloat(Value=&RandomRangeFloat(Minimum=-32, Maximum=-64))
    )
   )
   /SetVariables(Particles.SpawnOffset:Vector3f=&MakeVector(
     X=&RandomRangeFloat(Minimum=-20, Maximum=20)
     Y=&RandomRangeFloat(Minimum=0, Maximum=0)
     Z=&RandomRangeFloat(Minimum=-20, Maximum=10)
    )) {isSetParameters=true, module=SetVariables_#1}
   &SkeletalMeshLocation(
    Interpolate Mesh Sampling=false
    MaskSpawnGroup=false
    Mesh Sampling Type=::NewEnumerator5
    Position Sampling=::NewEnumerator1
    Sampled Position Offset=$Particles.SpawnOffset
    Skeletal Mesh=$User.SkeletalMesh
    Socket Sampling Mode=::NewEnumerator5
    Source Movement Scale:Vector3f=<0.05 0.05 0.05>
   )
   &DynamicMaterialParameters(
    Index 0 Param 2=&UniformRangedFloat(Minimum=0, Maximum=1)
    Index 0 Param 3=1
    Index 0 Param 4=&UniformRangedFloat(Minimum=1, Maximum=2)
    Module.Param0WriteEnabled=false
    Module.Param1WriteEnabled=true
    Module.Param2WriteEnabled=false
    Module.Param3WriteEnabled=true
   ) {module=DynamicMaterialParameters002}
   &AccelerationForce(Acceleration=<0 0 50>)
   &InheritVelocity(Module.Limit Speed=true, Velocity Limit=200)
   &ApplyInitialForces(Apply Rotational Force to Rotational Velocity=false)
   &CameraOffset(Camera Offset Amount=&RandomRangeFloat(Minimum=-2.5, Maximum=0))
   &AddVelocity(Velocity=<0 0 100>)
  ]
  particleUpdateScript [
   &ParticleState()
   &ScaleColor(
    Scale Mode=::NewEnumerator1
    Scale RGBA=&Vector4FromCurve(Vector4Curve=DataInterface() {propertyValues {
       AlphaCurve {keys [<at 0.0024 0>,<at 0.4194 1>,<at 0.457 1>,<at 1 0>]}
       BlueCurve {keys [<at 0.3836 0.0208>]}
       GreenCurve {keys [<at 0.3836 0.0208>]}
       RedCurve {keys [<at 0.3836 0.0208>]}
      }})
    ScaleRGBA=true
   )
   /SetVariables(Particles.SubImageIndex=&Add_Float(
     A=&FloatFromCurve(FloatCurve=DataInterface() {propertyValues {Curve {keys [<at 0 0>,<at 1 3>]}}})
     B=&RandomRangeFloat(Minimum=8, Maximum=30)
    )) {isSetParameters=true, module=SetVariables_#2}
   &ScaleSpriteSize(Scale Factor=&Vector2DFromFloat(Value=&FloatFromCurve(FloatCurve=DataInterface() {propertyValues {Curve {keys [
          {arriveTangent=7.5764, interpMode=RCIM_Cubic, leaveTangent=7.5765, tangentMode=RCTM_User, time=0, value=0.1}
          {arriveTangent=0.1045, interpMode=RCIM_Cubic, leaveTangent=0.1045, tangentMode=RCTM_User, time=0.25, value=1}
          {arriveTangent=-0.8944, interpMode=RCIM_Cubic, leaveTangent=-0.8944, tangentMode=RCTM_User, time=1, value=0.8}
         ]}}})))
   &DynamicMaterialParameters(
    Index 0 Param 1=&FloatFromCurve(FloatCurve=DataInterface() {propertyValues {Curve {keys [{arriveTangent=3.6282, interpMode=RCIM_Cubic, leaveTangent=3.6282, tangentMode=RCTM_User, time=0, value=0.5}, <at 0.5 1>]}}})
    Index 0 Param 2=&RandomRangeFloat(Minimum=0.33, Maximum=1)
    Index 0 Param 3=&FloatFromCurve(FloatCurve=DataInterface() {propertyValues {Curve {keys [
         {arriveTangent=0.1649, interpMode=RCIM_Cubic, leaveTangent=0.1649, tangentMode=RCTM_User, time=0, value=1.25}
         {arriveTangent=2.2011, interpMode=RCIM_Cubic, leaveTangent=2.2011, tangentMode=RCTM_User, time=1, value=2}
        ]}}})
    Index 0 Param 4=1
    Module.Param0WriteEnabled=true
    Module.Param1WriteEnabled=true
    Module.Param2WriteEnabled=true
    Module.Param3WriteEnabled=true
   ) {module=DynamicMaterialParameters001}
   &Drag(Drag=4, Use Linear Drag=true, Use Rotational Drag=false)
   &SolveForcesAndVelocity(Clamp Velocity=false, Limit Acceleration=false)
  ]
  rendererClasses [&NiagaraSpriteRendererProperties]
  renderers [{rendererClass=&NiagaraSpriteRendererProperties}]
 }
 Flames_Tip {
  emitterUpdateScript [
   &EmitterState(
    Inactive Response=::NewEnumerator1
    Life Cycle Mode=::NewEnumerator1
    Loop Behavior=::NewEnumerator1
    Loop Duration Mode=::NewEnumerator1
    Reset Age On Awaken=true
    Scalability Mode=::NewEnumerator1
    UseLoopDelay=false
   )
   &SpawnBurst_Instantaneous(Use Spawn Probability=false, UseLoopCountLimit=false)
  ]
  particleSpawnScript [
   &InitializeParticle(
    Mesh Scale=<1 0.9 0.75>
    Module.Write Color=true
    Module.Write Lifetime=true
    Module.Write Mass=true
    Module.Write Material Random=true
    Module.Write Position=false
    Module.Write Scale=true
    Module.Write SpriteRotation=false
    Module.Write SpriteSize=false
    Position=&SimulationPosition()
    Sprite Size=<5 5>
   )
   &DynamicMaterialParameters(
    Index 0 Param 1=&RandomRangeFloat(Minimum=5, Maximum=10)
    Index 0 Param 2=5
    Index 0 Param 3=30
    Module.Param0WriteEnabled=true
    Module.Param1WriteEnabled=true
    Module.Param2WriteEnabled=true
    Module.Param3WriteEnabled=true
   )
   &InitialMeshOrientation(Mesh Axis To Orient=<0 0 0>, Mesh Orientation Mode=::NewEnumerator2, Orientation Vector=<0 1 0>, Use Rotation Vector=true) {module=InitialMeshOrientation002}
   /Game/Effects/Niagara/Modules/Mesh/FN_MeshRenderer_v3/FN_MeshRenderer_v2(
    Color:LinearColor=<rgba 1 1 1 0.8>
    Create COlor Attribute=true
    Create Dynamic Param Attribute=true
    Create Position Attribute=true
    Create Scale Attribute=true
    DynamicMaterial:Vector4f=&MakeVector4(X=&RandomRangeFloat(Minimum=5, Maximum=10), Y=5, Z=10, W=0)
    Scale:Vector3f=<1 0.9 0.9>
   ) {module=FN_MeshRenderer_v2001}
  ]
  particleUpdateScript [
   &ParticleState(Kill Particles When Lifetime Has Elapsed=false, Let Infinitely Lived Particles Die When Emitter Deactivates=true)
   &SkeletalMeshLocation(
    Bone / Socket Index=1
    MaskSpawnGroup=false
    Mesh Sampling Type=::NewEnumerator5
    Sampled Position Offset=<0 0 -5>
    Skeletal Mesh=$User.SkeletalMesh
    Socket Sampling Mode=::NewEnumerator5
   )
   &InterpolateOverTime(
    Rate Of Change:NiagaraFloat=$User.Interpolate_Rate_of_Change
    Target Vector Value:NiagaraPosition=$Particles.Position
    Variable Type=::NewEnumerator2
   )
   /SetVariables(Particles.Interpolate Dynamic=&MakeFloat4FromVectorAndFloat("Vector (RGB)"=$StackContext.InterpolateOverTime.InterpolateOverTime_V3_.Moving Average, "Float (Alpha)"=$User.Interpolate_WPO)) {isSetParameters=true, module=SetVariables_#3}
   &SolveForcesAndVelocity(Clamp Velocity=false, Limit Acceleration=false)
   &ScaleColor(
    Scale Alpha=&Multiply_Float(A=&FNTimeOfDayScalar(Night Scale=1.5, Morning Scale=2, Day Scale=3, Evening Scale=2), B=0.2)
    ScaleA=true
    ScaleRGB=true
   )
   &CameraOffset(Camera Offset Amount=-10)
   /SetVariables(Particles.CameraOffsetBaseFlame=-10) {isSetParameters=true, module=SetVariables_#4}
  ]
  rendererClasses [&NiagaraMeshRendererProperties]
  renderers [
   {rendererClass=&NiagaraMeshRendererProperties, rendererIndex=0}
   {rendererClass=&NiagaraMeshRendererProperties, rendererIndex=1}
  ]
 }
 Flames_Top {
  emitterUpdateScript [
   &EmitterState(
    Inactive Response=::NewEnumerator1
    Life Cycle Mode=::NewEnumerator1
    Loop Behavior=::NewEnumerator1
    Loop Duration Mode=::NewEnumerator1
    Reset Age On Awaken=true
    Scalability Mode=::NewEnumerator1
    UseLoopDelay=false
   )
   &SpawnBurst_Instantaneous(Use Spawn Probability=false, UseLoopCountLimit=false)
  ]
  particleSpawnScript [
   &InitializeParticle(
    Mesh Scale=<1 1 0.9>
    Module.Write Color=true
    Module.Write Lifetime=true
    Module.Write Mass=true
    Module.Write Material Random=true
    Module.Write Position=false
    Module.Write Scale=true
    Module.Write SpriteRotation=false
    Module.Write SpriteSize=false
    Position=&SimulationPosition()
    Sprite Size=<5 5>
   )
   &DynamicMaterialParameters(
    Index 0 Param 1=&RandomRangeFloat(Minimum=5, Maximum=10)
    Index 0 Param 2=5
    Index 0 Param 3=30
    Module.Param0WriteEnabled=true
    Module.Param1WriteEnabled=true
    Module.Param2WriteEnabled=true
    Module.Param3WriteEnabled=true
   )
   &InitialMeshOrientation(Mesh Axis To Orient=<0 0 0>, Mesh Orientation Mode=::NewEnumerator2, Orientation Vector=<0 1 0>, Use Rotation Vector=true) {module=InitialMeshOrientation002}
   &FN_MeshRenderer_v2(
    Color:LinearColor=<rgba 1 1 1 0.6>
    Create COlor Attribute=true
    Create Dynamic Param Attribute=true
    Create Position Attribute=true
    Create Scale Attribute=true
    DynamicMaterial:Vector4f=&MakeVector4(X=&RandomRangeFloat(Minimum=5, Maximum=10), Y=5, Z=10, W=0)
    Scale:Vector3f=<0.5 1 1.3>
   ) {module=FN_MeshRenderer_v2001}
   &FN_MeshRenderer_v2(
    Color:LinearColor=<rgba 1 1 1 0.7>
    Create COlor Attribute=true
    Create Dynamic Param Attribute=true
    Create Position Attribute=true
    Create Scale Attribute=true
    DynamicMaterial:Vector4f=&MakeVector4(X=&RandomRangeFloat(Minimum=5, Maximum=10), Y=5, Z=20, W=0)
    Scale:Vector3f=<1 1 1>
   )
  ]
  particleUpdateScript [
   &ParticleState(Kill Particles When Lifetime Has Elapsed=false, Let Infinitely Lived Particles Die When Emitter Deactivates=true)
   &SkeletalMeshLocation(
    MaskSpawnGroup=false
    Mesh Sampling Type=::NewEnumerator5
    Sampled Position Offset=<0 1 1>
    Skeletal Mesh=$User.SkeletalMesh
    Socket Sampling Mode=::NewEnumerator5
   )
   &InterpolateOverTime(
    Rate Of Change:NiagaraFloat=$User.Interpolate_Rate_of_Change
    Target Vector Value:NiagaraPosition=$Particles.Position
    Variable Type=::NewEnumerator2
   )
   /SetVariables(Particles.Interpolate Dynamic=&MakeFloat4FromVectorAndFloat("Vector (RGB)"=$StackContext.InterpolateOverTime.InterpolateOverTime_V3_.Moving Average, "Float (Alpha)"=$User.Interpolate_WPO)) {isSetParameters=true, module=SetVariables_#3}
   &SolveForcesAndVelocity(Clamp Velocity=false, Limit Acceleration=false)
   &ScaleColor(
    Scale Alpha=&Multiply_Float(A=&FNTimeOfDayScalar(Night Scale=1.5, Morning Scale=2, Day Scale=3, Evening Scale=2), B=0.2)
    ScaleA=true
    ScaleRGB=true
   )
   &CameraOffset(Camera Offset Amount=-10)
   /SetVariables(Particles.CameraOffsetBaseFlame=10) {isSetParameters=true, module=SetVariables_#4}
  ]
  rendererClasses [&NiagaraMeshRendererProperties]
  renderers [
   {rendererClass=&NiagaraMeshRendererProperties, rendererIndex=0}
   {rendererClass=&NiagaraMeshRendererProperties, rendererIndex=1}
   {rendererClass=&NiagaraMeshRendererProperties, rendererIndex=2}
  ]
 }
}
summary {
 systemName=NS_Pickaxe_Raven_Quill_Desert_Idle
 emitters [
  {emitterName=Flame_Particles, rendererClasses [&NiagaraSpriteRendererProperties]}
  {emitterName=Flame_Ashes, rendererClasses [&NiagaraSpriteRendererProperties]}
  {emitterName=Flames_Top, rendererClasses [&NiagaraMeshRendererProperties]}
  {emitterName=Flames_Tip, rendererClasses [&NiagaraMeshRendererProperties]}
 ]
 userVariables [
  User.bIsFrontEndPreview {type=NiagaraBool}
  User.BladeSwing {type=NiagaraBool}
  User.Interpolate_Rate_of_Change {type=NiagaraFloat, defaultValue=7, description=""}
  User.Interpolate_WPO {type=NiagaraFloat, defaultValue=4, description=""}
  User.PlayerSpeed {type=NiagaraFloat, defaultValue=0, description=""}
  User.SkeletalMesh {
   type=NiagaraDataInterfaceSkeletalMesh
   defaultValue {
    struct=NiagaraExt_VariableValue_DataInterface
    value {dataInterface=&self:NiagaraDataInterfaceSkeletalMesh_1}
   }
  }
 ]
}
