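# Physics asset for the Raven Quill Desert Male pickaxe, a two-handed melee cosmetic with a chain of dangling dynamic bones.
# A capsule body covers the handle, and four similarly sized box bodies cover the dyn_mid_01 to dyn_mid_04 bones.
# Constraints link each dyn_mid bone to the previous one, starting at the handle, with identical tight swing and twist limits, so the chain can flex only slightly per joint.
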PhysicsAsset /BRCosmetics/Weapons/FORT_Melee/Pickaxe_Raven_Quill_Desert_Male/Meshes/Raven_Quill_Desert_Male_Axe_Physics
bodies {
 Melee_TwoHanded_Handle [{shapeType=Capsule, length=20.5304, radius=1.9633, center=<0.7208 -1.2624 27.2365>}]
 dyn_mid_01 [{shapeType=Box, extentX=4.4838, extentY=2.902, extentZ=1.4558, shapeName=dyn_mid_01_box, center=<2.1987 0 0>}]
 dyn_mid_02 [{shapeType=Box, extentX=4.4838, extentY=2.902, extentZ=1.4558, shapeName=dyn_mid_02_box, center=<2.1987 0 -0.3486>}]
 dyn_mid_03 [{shapeType=Box, extentX=4.4838, extentY=2.902, extentZ=1.4558, shapeName=dyn_mid_03_box, center=<2.1987 0 0>}]
 dyn_mid_04 [{shapeType=Box, extentX=4.4838, extentY=2.902, extentZ=1.4558, shapeName=dyn_mid_04_box, center=<2.1987 0 0>}]
}
constraints [
 {bone1Name=dyn_mid_01, bone2Name=Melee_TwoHanded_Handle, swing1LimitDegrees=5, swing2LimitDegrees=10, twistLimitDegrees=5}
 {bone1Name=dyn_mid_02, bone2Name=dyn_mid_01, swing1LimitDegrees=5, swing2LimitDegrees=10, twistLimitDegrees=5}
 {bone1Name=dyn_mid_03, bone2Name=dyn_mid_02, swing1LimitDegrees=5, swing2LimitDegrees=10, twistLimitDegrees=5}
 {bone1Name=dyn_mid_04, bone2Name=dyn_mid_03, swing1LimitDegrees=5, swing2LimitDegrees=10, twistLimitDegrees=5}
]
