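# Animation Blueprint for the Raven Quill Desert Male pickaxe that adds secondary physics motion to the axe's bones.
# It copies the pose from a mesh, runs a RigidBody node, then a series of four spring-driven Anim Dynamics nodes on dyn_mid_01 to 04.
# Each update, bIsInFrontEnd sets two alpha variables, so RigidBody is active in the front end and Anim Dynamics everywhere else.
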AnimBlueprint /BRCosmetics/Weapons/FORT_Melee/Pickaxe_Raven_Quill_Desert_Male/Meshes/Raven_Quill_Desert_Male_Axe_AnimBP
target_skeleton=&/Raven_Quill_Desert_Male_Axe_Skeleton
graphs [
 AnimGraph {nodes [
   n1 = RigidBody(ComponentPose=Local To Component(LocalPose=Copy Pose From Mesh()), Alpha=Get is_rban() -> (is_rban:double)) {
    bClampLinearTranslationLimitToRefPose=true
    bOverrideWorldGravity=true
    overridePhysicsAsset=&/Raven_Quill_Desert_Male_Axe_Physics
    simulationSpace=BaseBoneSpace
    alphaBoolBlend {blendOption=Linear}
    baseBoneRef {boneName=b_AR_Root}
    componentAppliedLinearAccClamp=<3000 3000 3000>
    componentLinearAccScale=<0.4 0.3 0.3>
    componentLinearVelScale=<0.4 0.3 0.3>
    overrideWorldGravity {z=-980}
   }
   n2 = Get is_anim_dynamic() -> (is_anim_dynamic:double)
   n3 = Anim Dynamics - Bone: dyn_mid_01(ComponentPose=@n1, Alpha=@n2, ExternalForce="0,0,0") {
    angularDampingOverride=0.99
    angularSpringConstant=750
    bAngularSpring=true
    bGravityOverrideInSimSpace=true
    bOverrideAngularDamping=true
    gravityScale=0.15
    alphaBoolBlend {blendOption=Linear}
    boundBone {boneName=dyn_mid_01}
    componentAppliedLinearAccClamp=<20000 20000 20000>
    componentLinearAccScale=<0.1 0.1 0.1>
    componentLinearVelScale=<0.9 0.9 0.9>
    gravityOverride {z=-980}
    physicsBodyDefinitions [{
      sphereCollisionRadius=0
      boundBone {boneName=dyn_mid_01}
      boxExtents=<5 15 15>
      constraintSetup {angularLimitsMax=<0 10 0>, angularLimitsMin=<0 -10 0>, angularTarget=<1 0 0>}
      localJointOffset=<3.236 0 0>
     }]
    relativeSpaceBone {boneName=b_AR_Root}
   }
   n4 = Anim Dynamics - Bone: dyn_mid_02(ComponentPose=@n3, Alpha=@n2, ExternalForce="0,0,0") {
    angularDampingOverride=0.99
    angularSpringConstant=1250
    bAngularSpring=true
    bGravityOverrideInSimSpace=true
    bOverrideAngularDamping=true
    gravityScale=0.25
    alphaBoolBlend {blendOption=Linear}
    boundBone {boneName=dyn_mid_02}
    componentAppliedLinearAccClamp=<20000 20000 20000>
    componentLinearAccScale=<0.1 0.1 0.1>
    componentLinearVelScale=<0.9 0.9 0.9>
    gravityOverride {z=-980}
    physicsBodyDefinitions [{
      sphereCollisionRadius=0
      boundBone {boneName=dyn_mid_02}
      boxExtents=<8 18 18>
      constraintSetup {angularLimitsMax=<5 20 5>, angularLimitsMin=<-5 -5 -5>, angularTarget=<1 0 0>}
      localJointOffset=<3.236 0 0>
     }]
    relativeSpaceBone {boneName=b_AR_Root}
   }
   n5 = Anim Dynamics - Bone: dyn_mid_03(ComponentPose=@n4, Alpha=@n2, ExternalForce="0,0,0") {
    angularDampingOverride=0.99
    angularSpringConstant=2500
    bAngularSpring=true
    bGravityOverrideInSimSpace=true
    bOverrideAngularDamping=true
    gravityScale=0.5
    alphaBoolBlend {blendOption=Linear}
    boundBone {boneName=dyn_mid_03}
    componentAppliedLinearAccClamp=<20000 20000 20000>
    componentLinearAccScale=<0.1 0.1 0.1>
    componentLinearVelScale=<0.5 0.5 0.5>
    gravityOverride {z=-980}
    physicsBodyDefinitions [{
      sphereCollisionRadius=0
      boundBone {boneName=dyn_mid_03}
      boxExtents=<8 10 10>
      constraintSetup {angularLimitsMax=<5 15 5>, angularLimitsMin=<-5 -5 -5>, angularTarget=<1 0 0>}
      localJointOffset=<3.236 0 0>
     }]
    relativeSpaceBone {boneName=b_AR_Root}
   }
   n6 = Anim Dynamics - Bone: dyn_mid_04(ComponentPose=@n5, Alpha=@n2, ExternalForce="0,0,0") {
    angularDampingOverride=0.99
    angularSpringConstant=2500
    bAngularSpring=true
    bGravityOverrideInSimSpace=true
    bOverrideAngularDamping=true
    gravityScale=0
    alphaBoolBlend {blendOption=Linear}
    boundBone {boneName=dyn_mid_04}
    componentAppliedLinearAccClamp=<20000 20000 20000>
    componentLinearAccScale=<0.1 0.1 0.1>
    componentLinearVelScale=<0.1 0.1 0.1>
    gravityOverride {z=-980}
    physicsBodyDefinitions [{
      sphereCollisionRadius=0
      boundBone {boneName=dyn_mid_04}
      boxExtents=<10 12 12>
      constraintSetup {angularLimitsMax=<5 30 5>, angularLimitsMin=<-5 -25 -5>, angularTarget=<1 0 0>}
      localJointOffset=<3.236 0 0>
     }]
    relativeSpaceBone {boneName=b_AR_Root}
   }
   Output Pose(Result=Component To Local(ComponentPose=@n6)) {layerGroup=None}
  ]}
 EventGraph {nodes [
   Event BlueprintUpdateAnimation()
   > n1 = Branch(Condition=Get bIsInFrontEnd())
   > n2 = Sequence()
   > Set is_rban(is_rban:double=1) -> (Output_Get:double)
   Set is_anim_dynamic(execute=@n2.then_1, is_anim_dynamic:double=0) -> (Output_Get:double)
   n3 = Sequence(execute=@n1.else)
   > Set is_rban(is_rban=0)
   Set is_anim_dynamic(execute=@n3.then_1, is_anim_dynamic=1)
  ]}
]
variables [
 is_anim_dynamic {type=number}
 is_rban {type=number}
]
